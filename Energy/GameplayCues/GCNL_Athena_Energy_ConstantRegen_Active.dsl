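; Looping gameplay cue for Athena's active constant energy regen, whose particle effect reacts to tactical sprint.
; On application it caches a component from the spawned particle array as MainNS and zeroes its "energy" float parameter.
; Ability tasks watch the target actor for the tactical sprint tag, setting the parameter to one when added and back to zero when removed.
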(event EventOnApplication(EitherVFX-preferthisone)
  (params (MyTarget :type Actor%20Object%20Reference) (Parameters :type Gameplay%20Cue%20Parameters%20Structure%20%28by%20ref%29) (ParticleComponents :type Array%20of%20FXSystem%20Component%20Object%20References) (AudioComponents :type Array%20of%20Audio%20Component%20Object%20References) (BurstCameraShakeInstance :type Camera%20Shake%20Base%20Object%20Reference) (BurstDecalInstance :type Decal%20Actor%20Object%20Reference))
  (bind _mainns (Variables|Default|GetMainNS))
  (bind _output_get (Variables|Default|SetMainNS (Utilities|Array|Get(acopy) ParticleComponents)))
  (Effects|Components|ParticleSystem|SetFloatParameter _output_get "energy" 0.0)
  (|CallParentFunction MyTarget Parameters ParticleComponents AudioComponents BurstCameraShakeInstance BurstDecalInstance)
  (Ability|Tasks|WaitGameplayTagAddToActor MyTarget (Variables|Default|GetTacticalSprintTag)
    (:then
      (Ability|Tasks|WaitGameplayTagRemoveFromActor MyTarget (Variables|Default|GetTacticalSprintTag)
        (:Removed
          (Effects|Components|ParticleSystem|SetFloatParameter _mainns "energy" 0.0))))
    (:Added
      (Effects|Components|ParticleSystem|SetFloatParameter _mainns "energy" 1.0))))

(fn ConstructionScript)
