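; Looping gameplay cue for Slap SMG constant energy regen that adds visual and audio feedback during tactical sprint.
; On application it caches the first particle component and starts async watchers for the tactical sprint tag on the target.
; Tag changes toggle the particle "energy" parameter and fade the sprint loop sound, spawned only if no component exists.
; On removal it ends both watchers, then fades and destroys the audio component.
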(event EventOnApplication(EitherVFX-preferthisone)
  (params (MyTarget :type Actor%20Object%20Reference) (Parameters :type Gameplay%20Cue%20Parameters%20Structure%20%28by%20ref%29) (ParticleComponents :type Array%20of%20FXSystem%20Component%20Object%20References) (AudioComponents :type Array%20of%20Audio%20Component%20Object%20References) (BurstCameraShakeInstance :type Camera%20Shake%20Base%20Object%20Reference) (BurstDecalInstance :type Decal%20Actor%20Object%20Reference))
  (bind _mainns (Variables|Default|GetMainNS))
  (bind _output_get (Variables|Default|SetMainNS (Utilities|Array|Get(acopy) ParticleComponents)))
  (Effects|Components|ParticleSystem|SetFloatParameter _output_get "energy" 0.0)
  (|CallParentFunction MyTarget Parameters ParticleComponents AudioComponents BurstCameraShakeInstance BurstDecalInstance)
  (bind _asynctaskproxy (Ability|Tasks|WaitGameplayTagAddToActor MyTarget (Variables|Default|GetTacticalSprintTag))
    (:then
      (Variables|Default|SetTagAddWatcher _asynctaskproxy)
      (bind _asynctaskproxy_2 (Ability|Tasks|WaitGameplayTagRemoveFromActor MyTarget (Variables|Default|GetTacticalSprintTag))
        (:then
          (Variables|Default|SetTagRemoveWatcher _asynctaskproxy_2))
        (:Removed
          (Effects|Components|ParticleSystem|SetFloatParameter _mainns "energy" 0.0)
          (bind _sprintaudiocomponent (Variables|Default|GetSprintAudioComponent)
            (:then
              (if (Audio|Components|Audio|IsPlaying _sprintaudiocomponent)
                (Audio|Components|Audio|FadeOut _sprintaudiocomponent 0.5 0.0 "Linear" :!class /Script/Engine.AudioComponent)))
            (:else)))))
    (:Added
      (Effects|Components|ParticleSystem|SetFloatParameter _mainns "energy" 1.0)
      (Variables|Default|GetSprintAudioComponent
        (:then)
        (:else
          (bind _returnvalue (Audio|SpawnSoundAttached (Variables|Default|GetSprintLoopSound) (Variables|/Script/Engine.Actor|GetRootComponent MyTarget)))
          (bind _output_get_2 (Variables|Default|SetSprintAudioComponent _returnvalue))
          (Audio|Components|Audio|FadeIn _output_get_2 0.5 1.0 0.0 "Linear" :!class /Script/Engine.AudioComponent))))))

(event EventOnRemoval(EitherVFX-preferthisversion)
  (params (MyTarget :type Actor%20Object%20Reference) (Parameters :type Gameplay%20Cue%20Parameters%20Structure%20%28by%20ref%29) (ParticleComponents :type Array%20of%20FXSystem%20Component%20Object%20References) (AudioComponents :type Array%20of%20Audio%20Component%20Object%20References) (BurstCameraShakeInstance :type Camera%20Shake%20Base%20Object%20Reference) (BurstDecalInstance :type Decal%20Actor%20Object%20Reference))
  (|CallParentFunction MyTarget Parameters ParticleComponents AudioComponents BurstCameraShakeInstance BurstDecalInstance)
  (Ability|Async|EndAction (Variables|Default|GetTagAddWatcher))
  (Ability|Async|EndAction (Variables|Default|GetTagRemoveWatcher))
  (bind _sprintaudiocomponent (Variables|Default|GetSprintAudioComponent)
    (:then
      (Audio|Components|Audio|FadeOut _sprintaudiocomponent 0.3 0.0 "Linear" :!class /Script/Engine.AudioComponent)
      (Components|DestroyComponent _sprintaudiocomponent 0))
    (:else)))

(fn ConstructionScript)
